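# Material function for fish gadget assets that builds a simple two-tone surface layer and outputs it as material attributes.
# The scalar Gmap input is the shared alpha that linearly blends the 01 and 02 variants of base color, specular and roughness.
# Metalness passes straight through unblended, and only these four attributes are set.
# Every input uses its preview value as a default, so each can be left unconnected.
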MaterialFunction /Game/Gadgets/Assets/Fish/Materials/MF_SimpleFishLayer
graph {nodes [
  n1 = FunctionInput() {bUsePreviewValueAsDefault=true, inputName=BaseColor02, sortPriority=1, previewValue=<1 1 1 1>}
  n2 = FunctionInput() {bUsePreviewValueAsDefault=true, inputName=Gmap, inputType=FunctionInput_Scalar, sortPriority=20}
  n3 = FunctionInput() {bUsePreviewValueAsDefault=true, inputName=Specular01, inputType=FunctionInput_Scalar, sortPriority=3, previewValue=<0.5 0 0 1>}
  n4 = FunctionInput() {bUsePreviewValueAsDefault=true, inputName=Specular02, inputType=FunctionInput_Scalar, sortPriority=4, previewValue=<0.5 0 0 1>}
  n5 = FunctionInput() {bUsePreviewValueAsDefault=true, inputName=Roughness01, inputType=FunctionInput_Scalar, sortPriority=5, previewValue=<0.2 0 0 1>}
  n6 = FunctionInput() {bUsePreviewValueAsDefault=true, inputName=Roughness02, inputType=FunctionInput_Scalar, sortPriority=6, previewValue=<0.2 0 0 1>}
  n7 = SetMaterialAttributes!(
   MaterialAttributes
   BaseColor=LinearInterpolate(A=FunctionInput() {bUsePreviewValueAsDefault=true, inputName=BaseColor01}, B=@n1, Alpha=@n2)
   Metallic=FunctionInput() {bUsePreviewValueAsDefault=true, inputName=Metalness, inputType=FunctionInput_Scalar, sortPriority=2}
   Specular=LinearInterpolate(A=@n3, B=@n4, Alpha=@n2)
   Roughness=LinearInterpolate(A=@n5, B=@n6, Alpha=@n2)
  ) {attributeSetTypes [
    #1
    #2
    #3
    #4
   ]}
  FunctionOutput(None=@n7)
 ]}
