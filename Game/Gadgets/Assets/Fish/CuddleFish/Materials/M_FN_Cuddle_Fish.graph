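# Master material for the skinned Cuddle Fish gadget, with optional Slurp and Galaxy skin effects.
# A switch picks texture inputs or three mask-blended MF_SimpleFishLayer layers, then an eye mask blends in eye color and emissive.
# Body emissive uses UVs or panned pre-skinned position with a view-facing term.
# World Position Offset is a UV1 sine wave weighted by vertex red; the panned-texture wave branch is unconnected.
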Material /Game/Gadgets/Assets/Fish/CuddleFish/Materials/M_FN_Cuddle_Fish
graph {
 properties {
  bUseMaterialAttributes=true
  bUsedWithMeshDeformer=true
  bUsedWithSkeletalMesh=true
  pixelDepthOffsetMode=PDOM_Legacy
  usedShadingModels=Default Lit
 }
 nodes [
  n1 = VertexColor()
  n2 = Add(A=@n1.A)
  n3 = ComponentMask(None=VectorParameter() {group=WPO, parameterName=TextureWPO_UVtiling, defaultValue=<rgba 1 0 0 0>}.RGB) {g=true, r=true}
  n4 = VectorParameter() {group=WPO, parameterName=TextureWPO_Speed, defaultValue=<rgba 1 0 0 0>}
  n5 = Add(A=ComponentMask(None=@n4.RGB) {r=true}, B=LinearInterpolate(Alpha=@n1.A) {constA=0.15, constB=0.35})
  n6 = ComponentMask(None=@n4.RGB) {g=true}
  n7 = TextureSample(UVs=Panner(Coordinate=Multiply(A=@n2, B=@n3), Speed=AppendVector(A=@n5, B=@n6)) {speedX=0.15}) {texture=/Game/Global/Textures/Default/Masks_and_Gradients/WeaponGradients/T_Gradient_DesertSky_Color}
  n8 = ScalarParameter() {defaultValue=1, group=WPO, parameterName=WPO_HeightX}
  n9 = ScalarParameter() {defaultValue=1, group=WPO, parameterName=WPO_HeightY}
  AppendVector(
   A=AppendVector(A=Multiply(A=@n7.R, B=@n8) {constB=0.5}, B=Multiply(A=@n7.R, B=@n9) {constB=0.5})
   B=Multiply(A=@n7.R, B=ScalarParameter() {defaultValue=1, group=WPO, parameterName=WPO_HeightZ}) {constB=0.6}
  )
  LinearInterpolate(
   A=ScalarParameter() {defaultValue=1, group=WPO, parameterName=WPO_Wave_Intensity_Close}
   B=ScalarParameter() {defaultValue=1, group=WPO, parameterName=WPO_Wave_Intenstiy}
  )
  n10 = ComponentMask(None=TextureCoordinate() {coordinateIndex=1}) {g=true}
  n11 = ScalarParameter() {defaultValue=2.5, group=WPO, parameterName=WPO_Tile_Rate}
  n12 = Time()
  n13 = ScalarParameter() {defaultValue=0.5, group=WPO, parameterName=WPO_Speed}
  n14 = Sine(None=Add(A=Multiply(A=@n10, B=@n11), B=Multiply(A=@n12, B=@n13)))
  Power(Base=@n14, Exp=ScalarParameter() {defaultValue=1, group=WPO, parameterName=WPO_Falloff})
  n15 = TextureSampleParameter2D() {group=Global, parameterName=Mask, samplerSource=SSM_Wrap_WorldGroupSettings, texture=&T_BlacksRGB}
  n16 = PreSkinnedPosition()
  n17 = TextureSampleParameter2D() {
   desc="R =SpecMask, G= Metallic, B + Roughness"
   group=TextureInputs
   parameterName=SpecularMasks
   samplerType=SAMPLERTYPE_Masks
   texture=&T_Fortnite_Default_S
  }
  n18 = SetMaterialAttributes!(
   MaterialAttributes
   Base Color=TextureSampleParameter2D() {group=TextureInputs, parameterName=Diffuse, texture=/Engine/EngineMaterials/DefaultDiffuse}.RGB
   Metallic=@n17.G
   Specular=@n17.G
   Roughness=@n17.G
  ) {attributeSetTypes [
    #1
    #2
    #3
    #4
   ]}
  n19 = MaterialFunctionCall!(
   BaseColor01=VectorParameter() {group=Layer01, parameterName=BaseColor01Layer01, defaultValue=<rgba 0.1667 0.1667 0.1667 1>}.RGB
   BaseColor02=VectorParameter() {group=Layer01, parameterName=BaseColor02Layer01, defaultValue=<rgba 0.7708 0.7708 0.7708 1>}.RGB
   Metalness=ScalarParameter() {defaultValue=0.5, group=Layer01, parameterName=MetalnessLayer01}
   Specular01=ScalarParameter() {defaultValue=0.5, group=Layer01, parameterName=Specular01Layer01}
   Specular02=ScalarParameter() {defaultValue=0.5, group=Layer01, parameterName=Specular02Layer01}
   Roughness01=ScalarParameter() {defaultValue=0.3, group=Layer01, parameterName=Roughness01Layer01}
   Roughness02=ScalarParameter() {defaultValue=0.3, group=Layer01, parameterName=Roughness02Layer01}
   Gmap=@n15.A
  ) {materialFunction=/Game/Gadgets/Assets/Fish/Materials/MF_SimpleFishLayer}
  n20 = MaterialFunctionCall(
   BaseColor01=VectorParameter() {group=Layer02, parameterName=BaseColor01Layer02, defaultValue=<rgba 0.1251 0.2448 0.1051 1>}.RGB
   BaseColor02=VectorParameter() {group=Layer02, parameterName=BaseColor02Layer02, defaultValue=<rgba 0.2074 0.4948 0.2196 1>}.RGB
   Metalness=ScalarParameter() {defaultValue=0.5, group=Layer02, parameterName=MetalnessLayer02}
   Specular01=ScalarParameter() {defaultValue=0.5, group=Layer02, parameterName=Specular01Layer02}
   Specular02=ScalarParameter() {defaultValue=0.5, group=Layer02, parameterName=Specular02Layer02}
   Roughness01=ScalarParameter() {defaultValue=0.3, group=Layer02, parameterName=Roughness01Layer02}
   Roughness02=ScalarParameter() {defaultValue=0.3, group=Layer02, parameterName=Roughness02Layer02}
   Gmap=@n15.A
  ) {materialFunction=&MF_SimpleFishLayer}
  n21 = VectorParameter() {group=Layer03, parameterName=BaseColor01Layer03, defaultValue=<rgba 0.0824 0.094 0.2969 1>}
  n22 = VectorParameter() {group=Layer03, parameterName=BaseColor02Layer03, defaultValue=<rgba 0.1249 0.3853 0.7188 1>}
  n23 = ScalarParameter() {defaultValue=0.5, group=Layer03, parameterName=MetalnessLayer03}
  n24 = ScalarParameter() {defaultValue=0.5, group=Layer03, parameterName=Specular01Layer03}
  n25 = ScalarParameter() {defaultValue=0.5, group=Layer03, parameterName=Specular02Layer03}
  n26 = ScalarParameter() {defaultValue=0.3, group=Layer03, parameterName=Roughness01Layer03}
  n27 = ScalarParameter() {defaultValue=0.3, group=Layer03, parameterName=Roughness02Layer03}
  n28 = StaticSwitchParameter(
   True=@n18
   False=BlendMaterialAttributes(
    A=BlendMaterialAttributes(A=@n19.Result, B=@n20.Result, Alpha=@n15.R)
    B=MaterialFunctionCall(
     BaseColor01=@n21.RGB
     BaseColor02=@n22.RGB
     Metalness=@n23
     Specular01=@n24
     Specular02=@n25
     Roughness01=@n26
     Roughness02=@n27
     Gmap=@n15.A
    ) {materialFunction=&MF_SimpleFishLayer}.Result
    Alpha=@n15.G
   )
  ) {group=Global, parameterName=UseTextureInputsForBaseMaterial}
  n29 = TextureSampleParameter2D(UVs=TextureCoordinate() {coordinateIndex=1}) {group=Global, parameterName=EyeTexture, samplerSource=SSM_Wrap_WorldGroupSettings, texture=/Engine/EngineMaterials/BlendFunc_DefBlend}
  n30 = SetMaterialAttributes(Base Color=@n29.RGB, Metallic=Constant(), Specular=Constant() {r=0.5}, Roughness=Constant() {r=0.1}) {attributeSetTypes [
    #1
    #2
    #3
    #4
   ]}
  n31 = Add(
   A=TextureCoordinate()
   B=ComponentMask(None=VectorParameter() {parameterName="EyeMaskPosition(Do Not Touch)", defaultValue=<rgba 0.5 -0.5 0 1>}.RGB) {g=true, r=true}
  )
  n32 = AppendVector(
   A=ScalarParameter() {defaultValue=0.1, parameterName="X_EyeMask_Size(Do Not Touch)"}
   B=ScalarParameter() {defaultValue=0.1, parameterName="Y_EyeMask_Size(Do Not Touch)"}
  )
  n33 = StaticSwitchParameter(
   True=Clamp(None=MaterialFunctionCall(UVs=@n31, Texture Scale=@n32) {materialFunction=&ScaleUVsByCenter}.0-1 mask)
   False=@n15.B
  ) {group=Global, parameterName=UseTextureInputsForBaseMaterial}
  n34 = TextureSampleParameter2D(UVs=TextureCoordinate()) {group=Global, parameterName=EmissiveTexture, texture=/Game/Characters/Fortress_Character/Textures/T_All_White}
  n35 = VectorParameter() {group=Global, parameterName=EmissiveColor, defaultValue=<rgba 0 0 0 1>}
  n36 = VertexInterpolator(VS=@n16)
  n37 = Add(A=ComponentMask(None=@n36.PS) {g=true, r=true}, B=Multiply(A=ComponentMask(None=@n36.PS) {b=true}, B=Constant() {r=0.5}))
  n38 = Constant() {r=20}
  n39 = TextureSampleParameter2D(UVs=Panner(Coordinate=Divide(A=@n37, B=@n38), Speed=ScalarParameter() {defaultValue=-0.05, group=Global, parameterName=EmissiveTexturePanSpeed})) {group=Global, parameterName=EmissiveTexture, texture=&T_All_White}
  n40 = Power(Base=DotProduct(A=PixelNormalWS(), B=CameraVectorWS())) {constExponent=4}
  n41 = StaticSwitchParameter(
   True=Multiply(A=Multiply(A=@n34.RGB, B=@n35.RGB), B=ScalarParameter() {defaultValue=1, parameterName=EmissiveStrength})
   False=Multiply(A=Multiply(A=@n39.R, B=@n35.RGB), B=Multiply(A=@n40, B=@n15.A))
  ) {defaultValue=true, parameterName=UseUVForEmissive}
  n42 = Multiply(A=@n29.RGB, B=ScalarParameter() {parameterName=Eye_EmissiveStrength})
  n43 = TextureSampleParameter2D() {
   group=TextureInputs
   parameterName=Normal
   samplerSource=SSM_Wrap_WorldGroupSettings
   samplerType=SAMPLERTYPE_Normal
   texture=&FlatNormalBlue
  }
  n44 = LinearInterpolate(
   A=Constant3Vector()
   B=@n14
   Alpha=LinearInterpolate(B=ScalarParameter() {group=WPO, parameterName=WPO_EyeMovent}, Alpha=VertexColor().R) {constA=1, constB=0}
  )
  n45 = VectorParameter() {group=WPO, parameterName=WPO_Intenstiy, defaultValue=<rgba 0 0 0.8 1>}
  n46 = SetMaterialAttributes!(
   MaterialAttributes=SetMaterialAttributes!(MaterialAttributes=BlendMaterialAttributes(A=@n28, B=@n30, Alpha=@n33), Emissive Color=LinearInterpolate(A=@n41, B=@n42, Alpha=@n33)) {attributeSetTypes [#5]}
   Normal=@n43.RGB
   Custom Data 0=Constant() {r=1}
   Custom Data 1=ScalarParameter() {defaultValue=0.1, group=Global, parameterName=ClearCoatRoughness}
   World Position Offset=Transform(None=Multiply(A=@n44, B=@n45.RGB) {constB=80}) {transformSourceType=TRANSFORMSOURCE_Local}
  ) {attributeSetTypes [
    #6
    #7
    #8
    #9
   ]}
  n47 = StaticSwitchParameter(
   True=MaterialFunctionCall!(
    Alpha=@n15.R
    AO=@n15.A
    BubblesMask=@n15.G
    SlurpNormalDistortionMask=@n15.B
    PreskinnedPosition=@n16
    Preskinned_Normal=PreSkinnedNormal()
    Material=@n46
   ) {materialFunction=/Game/Weapons/WeaponSkins/MaterialFunction/MF_Slurp_Fish}.Result
   False=@n46
  ) {parameterName=UseSlurp}
  StaticSwitchParameter(
   True=MaterialFunctionCall!(
    In=@n47
    Warp Mask=StaticComponentMaskParameter(None=@n15.RGBA) {defaultG=true, group=Skin FX, parameterName=Galaxy Warp Mask Channel}
    GalaxyAlternateFresnel
    Alpha=StaticComponentMaskParameter(None=@n15.RGBA) {defaultR=true, group=Skin FX, parameterName=Galaxy Channel}
   ) {materialFunction=/Game/Weapons/WeaponSkins/MaterialFunction/Galaxy}.Result
   False=@n47
  ) {group=Skin FX, parameterName=Use Galaxy}
 ]
}
