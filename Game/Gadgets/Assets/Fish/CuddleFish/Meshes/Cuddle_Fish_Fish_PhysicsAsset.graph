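# Physics asset for the cuddlefish fish mesh, providing simulated bodies for its six tentacles.
# Capsule bodies cover the mouth and six four-bone tentacle chains, three per side.
# Each tentacle base hangs from the mouth with free swing and twist, while segments down the chain have tight angular limits, narrowest at the base joint.
# No other part of the body has a collision shape.
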PhysicsAsset /Game/Gadgets/Assets/Fish/CuddleFish/Meshes/Cuddle_Fish_Fish_PhysicsAsset
bodies {
 mouth [{shapeType=Capsule, length=0.505, radius=0.505, rotation=<rot 0 -90 0>}]
 tentacle1_l_01 [{shapeType=Capsule, length=2, radius=2, rotation=<rot 90 0 0>}]
 tentacle1_l_02 [{shapeType=Capsule, length=3, radius=2, center=<1 0 0>, rotation=<rot 90 0 0>}]
 tentacle1_l_03 [{shapeType=Capsule, length=3, radius=2, center=<1 0 0>, rotation=<rot 90 0 0>}]
 tentacle1_l_base [{shapeType=Capsule, length=4.854, radius=1.1446}]
 tentacle1_r_01 [{shapeType=Capsule, length=2, radius=2, rotation=<rot 90 0 0>}]
 tentacle1_r_02 [{shapeType=Capsule, length=3, radius=2, center=<1 0 0>, rotation=<rot 90 0 0>}]
 tentacle1_r_03 [{shapeType=Capsule, length=3, radius=2, center=<1 0 0>, rotation=<rot 90 0 0>}]
 tentacle1_r_base [{shapeType=Capsule, length=4.854, radius=1.1446}]
 tentacle2_l_01 [{shapeType=Capsule, length=2, radius=2, rotation=<rot 90 0 0>}]
 tentacle2_l_02 [{shapeType=Capsule, length=3, radius=2, center=<1 0 0>, rotation=<rot 90 0 0>}]
 tentacle2_l_03 [{shapeType=Capsule, length=3, radius=2, center=<1 0 0>, rotation=<rot 90 0 0>}]
 tentacle2_l_base [{shapeType=Capsule, length=4.854, radius=1.1446}]
 tentacle2_r_01 [{shapeType=Capsule, length=2, radius=2, rotation=<rot 90 0 0>}]
 tentacle2_r_02 [{shapeType=Capsule, length=3, radius=2, center=<1 0 0>, rotation=<rot 90 0 0>}]
 tentacle2_r_03 [{shapeType=Capsule, length=3, radius=2, center=<1 0 0>, rotation=<rot 90 0 0>}]
 tentacle2_r_base [{shapeType=Capsule, length=4.854, radius=1.1446}]
 tentacle3_l_01 [{shapeType=Capsule, length=2, radius=2, rotation=<rot 90 0 0>}]
 tentacle3_l_02 [{shapeType=Capsule, length=3, radius=2, center=<1 0 0>, rotation=<rot 90 0 0>}]
 tentacle3_l_03 [{shapeType=Capsule, length=3, radius=2, center=<1 0 0>, rotation=<rot 90 0 0>}]
 tentacle3_l_base [{shapeType=Capsule, length=4.854, radius=1.1446}]
 tentacle3_r_01 [{shapeType=Capsule, length=2, radius=2, rotation=<rot 90 0 0>}]
 tentacle3_r_02 [{shapeType=Capsule, length=3, radius=2, center=<1 0 0>, rotation=<rot 90 0 0>}]
 tentacle3_r_03 [{shapeType=Capsule, length=3, radius=2, center=<1 0 0>, rotation=<rot 90 0 0>}]
 tentacle3_r_base [{shapeType=Capsule, length=4.854, radius=1.1446}]
}
constraints [
 {bone1Name=tentacle1_l_02, bone2Name=tentacle1_l_01, swing1LimitDegrees=15, swing2LimitDegrees=15, twistLimitDegrees=5}
 {bone1Name=tentacle1_l_03, bone2Name=tentacle1_l_02, swing1LimitDegrees=15, swing2LimitDegrees=15, twistLimitDegrees=5}
 {bone1Name=tentacle2_l_02, bone2Name=tentacle2_l_01, swing1LimitDegrees=15, swing2LimitDegrees=15, twistLimitDegrees=5}
 {bone1Name=tentacle2_l_03, bone2Name=tentacle2_l_02, swing1LimitDegrees=15, swing2LimitDegrees=15, twistLimitDegrees=5}
 {bone1Name=tentacle3_l_02, bone2Name=tentacle3_l_01, swing1LimitDegrees=15, swing2LimitDegrees=15, twistLimitDegrees=5}
 {bone1Name=tentacle3_l_03, bone2Name=tentacle3_l_02, swing1LimitDegrees=15, swing2LimitDegrees=15, twistLimitDegrees=5}
 {bone1Name=tentacle1_r_02, bone2Name=tentacle1_r_01, swing1LimitDegrees=15, swing2LimitDegrees=15, twistLimitDegrees=5}
 {bone1Name=tentacle1_r_03, bone2Name=tentacle1_r_02, swing1LimitDegrees=15, swing2LimitDegrees=15, twistLimitDegrees=5}
 {bone1Name=tentacle2_r_02, bone2Name=tentacle2_r_01, swing1LimitDegrees=15, swing2LimitDegrees=15, twistLimitDegrees=5}
 {bone1Name=tentacle2_r_03, bone2Name=tentacle2_r_02, swing1LimitDegrees=15, swing2LimitDegrees=15, twistLimitDegrees=5}
 {bone1Name=tentacle3_r_02, bone2Name=tentacle3_r_01, swing1LimitDegrees=15, swing2LimitDegrees=15, twistLimitDegrees=5}
 {bone1Name=tentacle3_r_03, bone2Name=tentacle3_r_02, swing1LimitDegrees=15, swing2LimitDegrees=15, twistLimitDegrees=5}
 {bone1Name=tentacle1_l_01, bone2Name=tentacle1_l_base, swing1LimitDegrees=5, swing2LimitDegrees=5, twistLimitDegrees=5}
 {bone1Name=tentacle2_l_01, bone2Name=tentacle2_l_base, swing1LimitDegrees=5, swing2LimitDegrees=5, twistLimitDegrees=5}
 {bone1Name=tentacle3_l_01, bone2Name=tentacle3_l_base, swing1LimitDegrees=5, swing2LimitDegrees=5, twistLimitDegrees=5}
 {bone1Name=tentacle1_r_01, bone2Name=tentacle1_r_base, swing1LimitDegrees=5, swing2LimitDegrees=5, twistLimitDegrees=5}
 {bone1Name=tentacle2_r_01, bone2Name=tentacle2_r_base, swing1LimitDegrees=5, swing2LimitDegrees=5, twistLimitDegrees=5}
 {bone1Name=tentacle3_r_01, bone2Name=tentacle3_r_base, swing1LimitDegrees=5, swing2LimitDegrees=5, twistLimitDegrees=5}
 {
  bone1Name=tentacle1_l_base
  bone2Name=mouth
  swing1LimitDegrees=45
  swing1Motion=Free
  swing2LimitDegrees=45
  swing2Motion=Free
  twistLimitDegrees=45
  twistMotion=Free
 }
 {
  bone1Name=tentacle2_l_base
  bone2Name=mouth
  swing1LimitDegrees=45
  swing1Motion=Free
  swing2LimitDegrees=45
  swing2Motion=Free
  twistLimitDegrees=45
  twistMotion=Free
 }
 {
  bone1Name=tentacle3_l_base
  bone2Name=mouth
  swing1LimitDegrees=45
  swing1Motion=Free
  swing2LimitDegrees=45
  swing2Motion=Free
  twistLimitDegrees=45
  twistMotion=Free
 }
 {
  bone1Name=tentacle1_r_base
  bone2Name=mouth
  swing1LimitDegrees=45
  swing1Motion=Free
  swing2LimitDegrees=45
  swing2Motion=Free
  twistLimitDegrees=45
  twistMotion=Free
 }
 {
  bone1Name=tentacle2_r_base
  bone2Name=mouth
  swing1LimitDegrees=45
  swing1Motion=Free
  swing2LimitDegrees=45
  swing2Motion=Free
  twistLimitDegrees=45
  twistMotion=Free
 }
 {
  bone1Name=tentacle3_r_base
  bone2Name=mouth
  swing1LimitDegrees=45
  swing1Motion=Free
  swing2LimitDegrees=45
  swing2Motion=Free
  twistLimitDegrees=45
  twistMotion=Free
 }
]
